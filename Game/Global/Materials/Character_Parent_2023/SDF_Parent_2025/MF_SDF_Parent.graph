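# Material function that layers an animated bubbling-liquid-in-glass look onto incoming material attributes for character materials, overriding base color, roughness, specular and emissive.
# Two optional bubble layers each multiply two panned samples of an SDF texture, colored through MF_SDF_Color.
# Fresnel liquid gradients, darkening, source-light glow, glass rim and optional sun-vector fake reflection add depth.
# Tech Art masks limit the effect, with bubble opacity driven by object-relative height or mask.
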MaterialFunction /Game/Global/Materials/Character_Parent_2023/SDF_Parent_2025/MF_SDF_Parent
graph {nodes [
  Constant()
  n1 = GetMaterialAttributes(None=FunctionInput() {inputName=Result, inputType=FunctionInput_MaterialAttributes}) {attributeGetTypes [
    #1
    #2
    #3
    #4
   ]}
  n2 = SetMaterialAttributes!(
   MaterialAttributes=@n1.MaterialAttributes
   BaseColor=NamedRerouteUsage()
   Roughness=NamedRerouteUsage()
   Specular=NamedRerouteUsage()
   EmissiveColor=NamedRerouteUsage()
  ) {attributeSetTypes [
    #1
    #2
    #3
    #4
   ]}
  FunctionOutput(None=@n2) {outputName=SetMaterialAttributes}
  n3 = ScalarParameter() {defaultValue=-3.554, group=SDF Bubble, parameterName=Source Light - Height}
  n4 = SmoothStep(
   Min=@n3
   Max=Add(A=@n3, B=ScalarParameter() {defaultValue=15, group=SDF Bubble, parameterName=Source Light - Smoothnesss})
   Value=NamedRerouteUsage()
  )
  Multiply(
   A=@n4
   B=MaterialFunctionCall(Exponent=ScalarParameter() {defaultValue=1.1739, group=SDF Bubble, parameterName=Source Light - Fresnel Shape}) {materialFunction=&AxisAlignedFresnel}.Fresnel
  )
  n5 = ScalarParameter() {defaultValue=1.25, group=SDF Bubble, parameterName=Liquid - Color Intensity}
  n6 = VectorParameter() {group=SDF Bubble, parameterName=Liquid - Color 02, sortPriority=6, defaultValue=<rgba 0.1109 0.1146 0.0043 1>}
  n7 = VectorParameter() {group=SDF Bubble, parameterName=Liquid - Color 01, sortPriority=6, defaultValue=<rgba 0.5972 0.6172 0.0194 1>}
  n8 = OneMinus(None=Fresnel(ExponentIn=Constant() {r=1}))
  n9 = ScalarParameter() {defaultValue=4, group=SDF Bubble, parameterName=Liquid - Fresnel Power}
  n10 = LinearInterpolate(A=@n6.RGB, B=@n7.RGB, Alpha=Saturate(None=Power(Base=@n8, Exp=@n9)))
  NamedRerouteDeclaration(None=Multiply(A=Multiply(A=@n5, B=@n10), B=NamedRerouteUsage())) {name=Emissive Fresnel Color, nodeColor=<rgba 1 0 0.6562 1>}
  n11 = NamedRerouteUsage()
  n12 = TextureSample(
   UVs=Panner(
    Coordinate=@n11
    Time=Time() {period=66.6667}
    Speed=VectorParameter() {group=SDF Bubble, parameterName=SDF - Speed 01, defaultValue=<rgba -0.003 -0.2 0 0>}.RGB
   )
   Tex=NamedRerouteUsage()
  ) {samplerType=SAMPLERTYPE_Masks, texture=/Game/Global/Materials/Character_Parent_2023/SDF_Parent_2025/Textures/T_SDF_Bubble_01}
  n13 = TextureSample(
   UVs=Panner(
    Coordinate=@n11
    Time=Time() {period=50}
    Speed=VectorParameter() {group=SDF Bubble, parameterName=SDF - Speed 02, defaultValue=<rgba -0.0045 -0.5 0 0>}.RGB
   )
   Tex=NamedRerouteUsage()
  ) {samplerType=SAMPLERTYPE_Masks, texture=&T_SDF_Bubble_01}
  n14 = ScalarParameter() {defaultValue=1.366, group=SDF Bubble, parameterName=SDF - Distance}
  n15 = Saturate(None=Multiply(A=Multiply(A=@n12.R, B=@n13.R), B=@n14) {constB=2.5})
  n16 = Multiply(A=@n15, B=@n15)
  n17 = Multiply(A=@n16, B=@n16)
  n18 = ScalarParameter() {defaultValue=0.8, group=SDF Bubble, parameterName=SDF - Normal Intensity, sortPriority=6}
  n19 = MaterialFunctionCall!(Fresnel=@n17, Height=Multiply(A=OneMinus(None=@n15), B=@n18) {constB=20}) {materialFunction=&/MF_SDF_Color}
  n20 = TextureSample(
   UVs=Panner(
    Coordinate=NamedRerouteUsage()
    Time=Time() {period=45.4545}
    Speed=VectorParameter() {group=SDF Bubble, parameterName=Secondary - SDF - Speed 01, defaultValue=<rgba -0.006 -0.1 0 0>}.RGB
   )
   Tex=NamedRerouteUsage()
  ) {samplerType=SAMPLERTYPE_Masks, texture=&T_SDF_Bubble_01}
  n21 = Panner(
   Coordinate=NamedRerouteUsage()
   Time=Time() {period=100}
   Speed=VectorParameter() {group=SDF Bubble, parameterName=Secondary - SDF - Speed 02, defaultValue=<rgba -0.0066 -0.2 0 0>}.RGB
  )
  n22 = TextureSample(UVs=@n21, Tex=NamedRerouteUsage()) {samplerType=SAMPLERTYPE_Masks, texture=&T_SDF_Bubble_01}
  n23 = Saturate(None=Multiply(A=@n14, B=Multiply(A=@n20.R, B=@n22.R)) {constB=2.5})
  n24 = Multiply(A=@n23, B=@n23)
  n25 = Multiply(A=@n24, B=@n24)
  n26 = MaterialFunctionCall(Fresnel=@n25, Height=Multiply(A=OneMinus(None=@n23), B=@n18) {constB=20}) {materialFunction=&MF_SDF_Color}
  n27 = Saturate(None=OneMinus(None=@n25))
  NamedRerouteDeclaration(None=StaticSwitchParameter(True=LinearInterpolate(A=@n19.Bubble Colors, B=@n26.Bubble Colors, Alpha=@n27), False=@n19.Bubble Colors) {defaultValue=true, group=SDF Bubble - Global, parameterName=Use Secondary Layer?}) {name=SDF Color, nodeColor=<rgba 1 0 0.9375 1>}
  n28 = ComponentMask(None=AtmosphericLightVector()) {b=true, g=true, r=true}
  n29 = Saturate(None=DotProduct(A=@n28, B=ReflectionVectorWS()))
  n30 = Fresnel(ExponentIn=Constant() {r=1}) {baseReflectFraction=0}
  n31 = Multiply(
   A=Power(Base=Max(A=@n29, B=@n30), Exp=ScalarParameter() {defaultValue=5, group=SDF Bubble, parameterName=Fake Spec - Highlight Power}) {constExponent=20}
   B=ScalarParameter() {group=SDF Bubble, parameterName=Fake Spec - Intensity}
  )
  n32 = OneMinus(None=NamedRerouteUsage())
  n33 = NamedRerouteUsage()
  NamedRerouteDeclaration(None=Multiply(A=Multiply(A=@n31, B=@n32), B=SkyLightEnvMapSample(Direction=@n28, Roughness=@n33))) {name=Fake Spec, nodeColor=<rgba 0 1 0.7422 1>}
  n34 = ScalarParameter() {defaultValue=10, group=SDF Bubble, parameterName=Source Light - Intensity}
  NamedRerouteDeclaration(None=Max(A=Multiply(A=@n4, B=@n34))) {name=Source Light, nodeColor=<rgba 0.8984 1 0 1>}
  n35 = Constant() {r=0.6}
  n36 = Add(A=@n35, B=Constant() {r=1.1762})
  n37 = Sine(None=Fresnel())
  n38 = NamedRerouteUsage()
  n39 = ScalarParameter() {group=SDF Bubble, parameterName=Fresnel Glass - Height}
  n40 = Multiply(
   A=Multiply(A=SmoothStep(Min=@n35, Max=@n36, Value=@n37), B=Add(A=@n38, B=@n39))
   B=VectorParameter() {group=SDF Bubble, parameterName=Fresnel Glass - Color, defaultValue=<rgba 0.491 0.4452 0.162 1>}.RGB
  )
  n41 = ScalarParameter() {defaultValue=1, group=SDF Bubble, parameterName=Fresnel Glass - Intensity}
  NamedRerouteDeclaration(None=Saturate(None=Multiply(A=@n40, B=@n41))) {name=Glass Edge, nodeColor=<rgba 1 0 0.2812 1>}
  n42 = NamedRerouteDeclaration(None=AppendVector(
    A=FunctionInput() {inputName=Tech Art Mask, sortPriority=10}
    B=FunctionInput() {inputName=Tech Art Mask Alpha, inputType=FunctionInput_Scalar, sortPriority=20}
   )) {name=TA Mask, nodeColor=<rgba 0.0547 0 1 1>}
  NamedRerouteDeclaration(None=LinearInterpolate(
    B=StaticComponentMaskParameter(None=@n42) {defaultB=true, group=SDF Bubble - Global, parameterName=SDF Mask}
    Alpha=ScalarParameter() {defaultValue=1, parameterName=Use TA Mask?}
   ) {constA=1}) {name=SDF Mask, nodeColor=<rgba 1 0.4219 0 1>}
  NamedRerouteDeclaration(None=@n1.Specular) {name=Specular - In, nodeColor=<rgba 0 0.0625 1 1>}
  NamedRerouteDeclaration(None=@n1.BaseColor) {name=Base Color - In, nodeColor=<rgba 0.3828 1 0 1>}
  NamedRerouteDeclaration(None=@n1.EmissiveColor) {name=Emissive Color - In, nodeColor=<rgba 1 0 0.7031 1>}
  n43 = NamedRerouteUsage()
  n44 = ScalarParameter() {group=SDF Bubble, parameterName=Bubble Glass - Color Multipler}
  NamedRerouteDeclaration(None=LinearInterpolate(A=@n43, B=Multiply(A=@n43, B=@n44), Alpha=NamedRerouteUsage()) {constB=0}) {name=Base Color - Out, nodeColor=<rgba 0 1 0.9766 1>}
  n45 = NamedRerouteUsage()
  n46 = ScalarParameter() {defaultValue=1, group=SDF Bubble, parameterName=Glass - Roughness Multipler}
  NamedRerouteDeclaration(None=LinearInterpolate(A=@n45, B=Multiply(A=@n45, B=@n46), Alpha=NamedRerouteUsage()) {constB=0.1}) {name=Roughness - Out, nodeColor=<rgba 1 0.6562 0 1>}
  n47 = VectorNoise(World Position=AppendVector(A=@n21, B=Constant()))
  NamedRerouteDeclaration(None=Step(Y=ComponentMask(None=@n47) {r=true}, X=ScalarParameter() {parameterName=Cluster Amount})) {name=Cluster Mask, nodeColor=<rgba 0 0.3437 1 1>}
  n48 = NamedRerouteUsage()
  n49 = ScalarParameter() {defaultValue=1, group=SDF Bubble, parameterName=Glass - Specular Multipler}
  NamedRerouteDeclaration(None=LinearInterpolate(A=@n48, B=Multiply(A=@n48, B=@n49), Alpha=NamedRerouteUsage()) {constB=0.5}) {name=Specular - Out, nodeColor=<rgba 0.3359 0 1 1>}
  n50 = NamedRerouteUsage()
  n51 = StaticComponentMaskParameter(None=NamedRerouteUsage()) {defaultG=true, defaultR=true, group=SDF Bubble - Global, parameterName=SDF- Opacity Mask}
  n52 = ComponentMask(None=Subtract(A=WorldPosition().XYZ, B=ObjectPositionWS())) {b=true}
  n53 = ScalarParameter() {defaultValue=-300, group=SDF Bubble, parameterName="SDF Opacity - Top Down  - Height"}
  n54 = Saturate(None=Divide(A=Subtract(A=@n52, B=@n53), B=ScalarParameter() {defaultValue=300, group=SDF Bubble, parameterName="SDF Opacity - Top Down  - Contrast"}))
  n55 = Multiply(A=@n54, B=@n54)
  n56 = Multiply(A=@n55, B=@n55)
  n57 = Saturate(None=LinearInterpolate(
    A=ScalarParameter() {group=SDF Bubble, parameterName=SDF Opacity - Top Down - Min}
    B=ScalarParameter() {defaultValue=0.5, group=SDF Bubble, parameterName=SDF Opacity - Top Down - Max}
    Alpha=StaticSwitchParameter(True=@n51, False=@n56) {group=SDF Bubble - Global, parameterName=Use TA Mask for Bubble Opacity}
   ))
  n58 = ScalarParameter() {defaultValue=0.15, group=SDF Bubble, parameterName=SDF - Edge Fade - Step}
  n59 = ScalarParameter() {defaultValue=0.3, group=SDF Bubble, parameterName=SDF - Edge Fade - Smoothness}
  n60 = Fresnel(ExponentIn=ScalarParameter() {defaultValue=0.5, group=SDF Bubble, parameterName=SDF - Edge Fade - Fade Fresnel}) {baseReflectFraction=0}
  n61 = Multiply(
   A=ScalarParameter() {defaultValue=0.1, group=SDF Bubble, parameterName=Secondary SDF - Opacity, sortPriority=7}
   B=Saturate(None=OneMinus(None=@n17))
  ) {constB=0.6}
  n62 = Max(A=@n61, B=Multiply(A=ScalarParameter() {defaultValue=0.3, group=SDF Bubble, parameterName=Main SDF - Opacity, sortPriority=7}, B=@n27))
  n63 = Multiply(
   A=SmoothStep(Min=@n58, Max=Add(A=@n58, B=@n59), Value=OneMinus(None=@n60))
   B=Multiply(A=StaticSwitchParameter(True=@n62, False=@n61) {defaultValue=true, group=SDF Bubble - Global, parameterName=Use Secondary Layer?}) {constB=2}
  )
  n64 = LinearInterpolate(
   A=LinearInterpolate(A=NamedRerouteUsage(), B=NamedRerouteUsage(), Alpha=@n50)
   B=NamedRerouteUsage()
   Alpha=Multiply(A=@n50, B=Multiply(A=@n57, B=@n63))
  )
  n65 = LinearInterpolate(A=@n64, B=Multiply(A=@n64) {constB=0}, Alpha=NamedRerouteUsage())
  n66 = NamedRerouteUsage()
  n67 = StaticSwitchParameter(True=LinearInterpolate(A=@n65, B=Add(A=@n65, B=@n66), Alpha=NamedRerouteUsage()), False=@n65) {group=SDF Bubble - Global, parameterName=Use Fake Reflection}
  n68 = NamedRerouteUsage()
  NamedRerouteDeclaration(None=LinearInterpolate(A=NamedRerouteUsage(), B=Add(A=@n67, B=@n68), Alpha=NamedRerouteUsage())) {name=Emissive Color - Out, nodeColor=<rgba 0 1 0.0156 1>}
  n69 = StaticSwitchParameter(True=TextureCoordinate() {coordinateIndex=1}, False=TextureCoordinate()) {group=SDF Bubble, parameterName=Use UV2?}
  n70 = ComponentMask(None=Ceil(None=VectorParameter() {group=SDF Bubble, parameterName=SDF - Tiling, defaultValue=<rgba 7 7 0 0>}.RGB)) {g=true, r=true}
  NamedRerouteDeclaration(None=FunctionInput(Preview=Multiply(A=@n69, B=@n70)) {bUsePreviewValueAsDefault=true, inputName=UV, inputType=FunctionInput_Vector2}) {name=UV, nodeColor=<rgba 0.2188 0 1 1>}
  NamedRerouteDeclaration(None=@n1.Roughness) {name=Roughness In, nodeColor=<rgba 0.9219 1 0 1>}
  n71 = WorldPosition()
  n72 = ObjectPositionWS()
  n73 = OneMinus(None=ComponentMask(None=Subtract(A=@n71.XYZ, B=@n72)) {b=true})
  n74 = PreSkinnedLocalBounds()
  n75 = ComponentMask(None=@n74.Half Extents) {b=true}
  NamedRerouteDeclaration(None=Divide(A=Subtract(A=@n73, B=@n75), B=ComponentMask(None=@n74.Extents) {b=true})) {name=Preskinned Bounds, nodeColor=<rgba 0.4062 0 1 1>}
  n76 = ScalarParameter() {defaultValue=1.3793, group=SDF Bubble, parameterName=Liquid Darkening - Step}
  n77 = ScalarParameter() {defaultValue=1, group=SDF Bubble, parameterName=Liquid Darkening - Smoothness}
  n78 = NamedRerouteUsage()
  n79 = ScalarParameter() {defaultValue=1.87, group=SDF Bubble, parameterName=Liquid Darkening - Fresnel}
  n80 = Multiply(
   A=SmoothStep(Min=@n76, Max=Add(A=@n76, B=@n77), Value=OneMinus(None=@n78))
   B=OneMinus(None=MaterialFunctionCall(Exponent=@n79) {materialFunction=&AxisAlignedFresnel}.Fresnel)
  )
  n81 = ScalarParameter() {group=SDF Bubble, parameterName=Liquid Darkening - Intensity}
  NamedRerouteDeclaration(None=Saturate(None=Multiply(A=@n80, B=@n81))) {name=Liquid Darkening, nodeColor=<rgba 0.0547 0 1 1>}
  n82 = TextureObjectParameter() {
   group=SDF Bubble - Global
   parameterName=SDF Shape
   samplerType=SAMPLERTYPE_LinearGrayscale
   texture=&T_SDF_Bubble_01
  }
  NamedRerouteDeclaration(None=@n82) {name=SDF Texture, nodeColor=<rgba 1 0.5859 0 1>}
  NamedRerouteDeclaration(None=Step(Y=ComponentMask(None=@n47) {g=true}, X=ScalarParameter() {defaultValue=0.2023, parameterName=Cluster Amount 2})) {name=Cluster Mask 2, nodeColor=<rgba 1 0.6094 0 1>}
 ]}
